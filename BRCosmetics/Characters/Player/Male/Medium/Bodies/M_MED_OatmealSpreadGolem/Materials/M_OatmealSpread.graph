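# Character material for the OatmealSpreadGolem body, built on the 2023 character parent PreFX and PostFX functions.
# An optional static switch applies MF_Slurp, masked by PreFX M-mask channels, while restoring the PreFX World Position Offset.
# A second switch blends UV2 composite decals over that result through a selectable Tech Art Mask channel, green by default.
# The result feeds PostFX, with shading model set by material expression.
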Material /BRCosmetics/Characters/Player/Male/Medium/Bodies/M_MED_OatmealSpreadGolem/Materials/M_OatmealSpread
graph {
 properties {
  bUseMaterialAttributes=true
  bUsedWithClothing=true
  bUsedWithMeshDeformer=true
  bUsedWithMorphTargets=true
  bUsedWithSkeletalMesh=true
  shadingModel=MSM_FromMaterialExpression
  usedShadingModels=Default Lit | Subsurface | Clear Coat | Subsurface Profile | Cloth
 }
 nodes [
  n1 = MaterialFunctionCall() {materialFunction=&MF_CharacterParent_2023_PreFX}
  n2 = ComponentMask(None=@n1.M) {b=true}
  n3 = ComponentMask(None=@n1.M) {r=true}
  n4 = GetMaterialAttributes(None=@n1.Result) {attributeGetTypes [#1]}
  n5 = StaticSwitchParameter(
   True=SetMaterialAttributes!(
    MaterialAttributes=MaterialFunctionCall!(
     Alpha=@n2
     AO=@n3
     BubblesMask=@n2
     SlurpNormalDistortionMask=@n2
     Pre Skinned Position=@n1.Pre skinned Local Position
     Material=@n1.Pre skinned Local Position
    ) {materialFunction=/Game/Weapons/WeaponSkins/MaterialFunction/MF_Slurp}.Result
    World Position Offset=@n4.World Position Offset
   ) {attributeSetTypes [#1]}
   False=@n4.MaterialAttributes
  ) {group=04_Tech Art, parameterName=Oatmeal Use Slurp}
  n6 = MaterialFunctionCall!(Material=@n5, TechArtMask=@n1.Tech Art Mask) {materialFunction=/Game/Global/Materials/Character_Parent_2023/UV2Composite_Parent_2023/MF_UV2_Composite_2023}
  n7 = StaticComponentMaskParameter(None=@n1.Tech Art Mask) {defaultG=true, group=05_Tech Art Masks, parameterName=UV2 Composite Area}
  MaterialFunctionCall(
   Material=StaticSwitchParameter(True=BlendMaterialAttributes(A=@n5, B=@n6.Result, Alpha=@n7), False=@n5) {group=04_Tech Art, parameterName=Oatmeal Use Decals}
   M=@n1.M
   Pre Skinned Local Position=@n1.M
   Distance Blend=@n1.M
   Fresnel=@n1.M
  ) {materialFunction=&MF_CharacterParent_2023_PostFX}
 ]
}
