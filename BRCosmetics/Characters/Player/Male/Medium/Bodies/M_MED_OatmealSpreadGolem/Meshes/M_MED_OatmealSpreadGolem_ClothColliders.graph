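# Physics asset of cloth colliders for the OatmealSpreadGolem medium male player body cosmetic.
# It defines four bodies on the left and right thigh and calf bones, covering only the legs.
# It has no constraints between the bodies, so its name and layout point to cloth collision rather than ragdoll simulation.
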PhysicsAsset /BRCosmetics/Characters/Player/Male/Medium/Bodies/M_MED_OatmealSpreadGolem/Meshes/M_MED_OatmealSpreadGolem_ClothColliders
bodies {calf_l [], calf_r [], thigh_l [], thigh_r []}
constraints []
