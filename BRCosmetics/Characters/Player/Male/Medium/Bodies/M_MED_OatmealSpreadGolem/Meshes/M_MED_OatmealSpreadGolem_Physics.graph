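# Physics asset for the medium male OatmealSpreadGolem player body cosmetic, set up to simulate two short dangling strings at the hips.
# Left and right four-bone dyn_shortString chains of small capsules hang from the pelvis, each link constrained to its parent with limited swing and almost no twist.
# Larger capsules on the pelvis, thighs and calves cover only the lower body, apparently as collision for the strings.
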PhysicsAsset /BRCosmetics/Characters/Player/Male/Medium/Bodies/M_MED_OatmealSpreadGolem/Meshes/M_MED_OatmealSpreadGolem_Physics
bodies {
 calf_l [{shapeType=Capsule, length=34.5, radius=7, center=<-17.7026 -1.9577 -0.5552>, rotation=<rot -90 0 -0.0001>}]
 calf_r [{shapeType=Capsule, length=34.5, radius=7, center=<17.7026 1.5842 0.8815>, rotation=<rot 90 179.9204 179.9217>}]
 dyn_shortString_l_01 [{shapeType=Capsule, length=2, radius=1.5, shapeName=dyn_shortString_l_01_capsule, center=<1.6534 0 0>, rotation=<rot -90 180 180>}]
 dyn_shortString_l_02 [{shapeType=Capsule, length=2, radius=1.5, shapeName=dyn_shortString_l_02_capsule, center=<1.6534 0 0>, rotation=<rot -90 180 180>}]
 dyn_shortString_l_03 [{shapeType=Capsule, length=3, radius=1.5, shapeName=dyn_shortString_l_03_capsule, center=<1.6534 0 0>, rotation=<rot -90 180 180>}]
 dyn_shortString_l_04 [{shapeType=Capsule, length=3, radius=1.5, shapeName=dyn_shortString_l_04_capsule, center=<1.6534 0 0>, rotation=<rot -90 180 180>}]
 dyn_shortString_r_01 [{shapeType=Capsule, length=2, radius=1.5, shapeName=dyn_shortString_r_01_capsule, center=<1.6534 0 0>, rotation=<rot -90 180 180>}]
 dyn_shortString_r_02 [{shapeType=Capsule, length=2, radius=1.5, shapeName=dyn_shortString_r_02_capsule, center=<1.6534 0 0>, rotation=<rot -90 180 180>}]
 dyn_shortString_r_03 [{shapeType=Capsule, length=3, radius=1.5, shapeName=dyn_shortString_r_03_capsule, center=<1.6534 0 0>, rotation=<rot -90 180 180>}]
 dyn_shortString_r_04 [{shapeType=Capsule, length=3, radius=1.5, shapeName=dyn_shortString_r_04_capsule, center=<1.6534 0 0>, rotation=<rot -90 180 180>}]
 pelvis [{shapeType=Capsule, length=8, radius=16, center=<-10.7507 -6.6454 0>, rotation=<rot -90 -10.025 10.0249>}]
 thigh_l [{shapeType=Capsule, length=15.5, radius=12, center=<-16.3048 -2.7521 0.5848>, rotation=<rot 90 179.9899 179.9903>}]
 thigh_r [{shapeType=Capsule, length=15.5, radius=12, center=<16.3048 3.167 -0.5848>, rotation=<rot -90 -179.9802 179.9801>}]
}
constraints [
 {bone1Name=dyn_shortString_l_01, bone2Name=pelvis, swing1LimitDegrees=45, swing2LimitDegrees=45, twistLimitDegrees=0}
 {bone1Name=dyn_shortString_l_02, bone2Name=dyn_shortString_l_01, swing1LimitDegrees=45, swing2LimitDegrees=30, twistLimitDegrees=0}
 {bone1Name=dyn_shortString_l_03, bone2Name=dyn_shortString_l_02, swing1LimitDegrees=45, swing2LimitDegrees=45, twistLimitDegrees=1}
 {bone1Name=dyn_shortString_l_04, bone2Name=dyn_shortString_l_03, swing1LimitDegrees=45, swing2LimitDegrees=45, twistLimitDegrees=1}
 {bone1Name=dyn_shortString_r_01, bone2Name=pelvis, swing1LimitDegrees=45, swing2LimitDegrees=45, twistLimitDegrees=0}
 {bone1Name=dyn_shortString_r_02, bone2Name=dyn_shortString_r_01, swing1LimitDegrees=45, swing2LimitDegrees=30, twistLimitDegrees=0}
 {bone1Name=dyn_shortString_r_03, bone2Name=dyn_shortString_r_02, swing1LimitDegrees=45, swing2LimitDegrees=45, twistLimitDegrees=1}
 {bone1Name=dyn_shortString_r_04, bone2Name=dyn_shortString_r_03, swing1LimitDegrees=45, swing2LimitDegrees=45, twistLimitDegrees=1}
]
