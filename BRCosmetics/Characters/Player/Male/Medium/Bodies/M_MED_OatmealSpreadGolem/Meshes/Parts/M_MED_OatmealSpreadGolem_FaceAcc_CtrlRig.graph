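# Control Rig for the OatmealSpreadGolem face accessory that adjusts hat and headband fin bones as the head moves.
# A spherical pose reader on the head bone writes its falloff weight to the HeadUp variable.
# HeadUp then weights two Modify Transforms nodes, one posing hat fins and one headband fins, through AlphaInterp functions with interpolation and clamping off.
# The pose reader's debug drawing is left enabled.
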ControlRigBlueprint /BRCosmetics/Characters/Player/Male/Medium/Bodies/M_MED_OatmealSpreadGolem/Meshes/Parts/M_MED_OatmealSpreadGolem_FaceAcc_CtrlRig
graphs [
 RigVMFunctionLibrary {links [], nodes []}
 RigVMFunction_Sequence_1_ContainedGraph {nodes [
   n1 = RigVMFunctionEntryNode()
   RigVMFunction_Sequence = &RigVMFunction_Sequence(ExecuteContext=@n1.ExecuteContext)
   n2 = &RigVMFunction_Sequence(ExecuteContext=@RigVMFunction_Sequence.B)
   RigVMFunctionReturnNode(A=@RigVMFunction_Sequence.A, B=@n2.A, C=@n2.B)
  ]}
 RigVMModel {nodes [
   RigUnit_BeginExecution = &RigUnit_BeginExecution()
   RigVMFunction_Sequence = RigVMAggregateNode(ExecuteContext=@RigUnit_BeginExecution.ExecutePin)
   n1 = &RigUnit_SphericalPoseReader(
    ActiveRegionScaleFactors=(PositiveWidth=1,NegativeWidth=1,PositiveHeight=1,NegativeHeight=1)
    ActiveRegionSize=0.2
    Debug=(bDrawDebug=True,bDraw2D=False,bDrawLocalAxes=False,DebugScale=25,DebugSegments=20,DebugThickness=0.25)
    DriverAxis=(X=1,Y=0,Z=0)
    DriverItem=(Type=Bone,Name="head")
    FalloffRegionScaleFactors=(PositiveWidth=1.5,NegativeWidth=1,PositiveHeight=1,NegativeHeight=1)
    FalloffSize=0.32
    FlipHeightScaling=False
    FlipWidthScaling=False
    OptionalParentItem=(Type=Bone,Name="None")
    RotationOffset=(X=-64.889,Y=15.1763,Z=36.1456)
    ExecutePin=@RigVMFunction_Sequence.A
   )
   RigVMVariableNode(Value=@n1.OutputParam, ExecuteContext=@RigVMFunction_Sequence.B) {variable=HeadUp}
   AlphaInterp = &RigVMFunction_AlphaInterp(
    InterpSpeedDecreasing=20
    InterpSpeedIncreasing=20
    Scale=1
    bClampResult=False
    bInterpResult=False
    Value=RigVMVariableNode() {variable=HeadUp}.Value
   )
   &RigUnit_ModifyTransforms(
    ItemToModify="((Item=(Type=Bone,Name=\"dyn_finHat_top_r_01\"),Transform=(Rotation=(X=-0.225,Y=-0.6135,Z=-0.3976,W=0.6442),Translation=(X=0,Y=0,Z=0),Scale3D=(X=1,Y=1,Z=1))),(Item=(Type=Bone,Name=\"dyn_finHat_bot_r_base\"),Transform=(Rotation=(X=-0.212,Y=-0.554,Z=-0.1485,W=0. ...(truncated)"
    WeightMinimum=0
    Weight=@AlphaInterp.Result
    ExecutePin=@RigVMFunction_Sequence.C
   )
   n2 = &RigVMFunction_AlphaInterp(
    InterpSpeedDecreasing=20
    InterpSpeedIncreasing=20
    Scale=1
    bClampResult=False
    bInterpResult=False
    Value=RigVMVariableNode() {variable=HeadUp}.Value
   )
   > &RigUnit_ModifyTransforms(
    ItemToModify="((Item=(Type=Bone,Name=\"dyn_finHeadband_bot_r_base\"),Transform=(Rotation=(X=-0.1016,Y=0.3175,Z=0.4086,W=0.8496),Translation=(X=0,Y=0,Z=0),Scale3D=(X=1,Y=1,Z=1))),(Item=(Type=Bone,Name=\"dyn_finHeadband_mid_r_01\"),Transform=(Rotation=(X=0.1063,Y=0.2708,Z=0.2 ...(truncated)"
    WeightMinimum=0
    Weight=@n2.Result
   )
  ]}
]
hierarchy [
 root {loc=<0 0 0>, rot=<0 0 0 1>}
 pelvis {parent=root, loc=<0 -0.7071 93.1191>, rot=<0.0127 0.707 -0.0127 -0.707>}
 spine_01 {parent=pelvis, loc=<6.7494 -0.6388 0>, rot=<0 0 0.034 -0.9994>}
 spine_02 {parent=spine_01, loc=<7.252 -0.035 0>, rot=<0 0 0.0238 -0.9997>}
 spine_03 {parent=spine_02, loc=<6.0106 0 0>, rot=<0 0 0.0149 -0.9999>}
 spine_04 {parent=spine_03, loc=<8.1023 0 0>, rot=<0 0 -0.0132 -0.9999>}
 spine_05 {parent=spine_04, loc=<10.0142 0 0>, rot=<0 0 -0.0015 -1>}
 neck_01 {parent=spine_05, loc=<11.3138 0.1529 0>, rot=<0 0 -0.2003 -0.9797>}
 neck_02 {parent=neck_01, loc=<4.5627 0 0>, rot=<0 0 0.0504 -0.9987>}
 head {parent=neck_02, loc=<5.3914 0 0>, rot=<0 0 0.115 -0.9934>}
 dyn_finHat_top_l_01 {parent=head, loc=<4.0222 -3.1502 -11.22>, rot=<0.2065 -0.843 0.4539 -0.2016>}
 dyn_finHat_top_l_02 {parent=dyn_finHat_top_l_01, loc=<4.8072 -1.7497 0>, rot=<0.0408 -0.0187 0.0877 -0.9951>}
 dyn_finHat_mid_l_01 {parent=head, loc=<1.3405 -7.9556 -5.6279>, rot=<0.0782 -0.787 0.6085 -0.065>}
 dyn_finHat_mid_l_02 {parent=dyn_finHat_mid_l_01, loc=<4.8381 -2.2386 0.1203>, rot=<-0.0059 -0.0333 0.048 -0.9983>}
 dyn_finHat_top_r_01 {parent=head, loc=<4.0222 -3.1502 11.2203>, rot=<0.186 -0.4628 0.8432 -0.2005>}
 dyn_finHat_top_r_02 {parent=dyn_finHat_top_r_01, loc=<4.8169 1.7124 0.1893>, rot=<0.0015 -0.0112 -0.0886 -0.996>}
 dyn_finHat_mid_r_01 {parent=head, loc=<1.3405 -7.9556 5.628>, rot=<0.0502 -0.5869 0.8072 -0.0369>}
 dyn_finHat_mid_r_02 {parent=dyn_finHat_mid_r_01, loc=<4.696 2.5038 0.3362>, rot=<0.0435 -0.0259 -0.0823 -0.9953>}
 dyn_finHeadband_top_r_01 {parent=head, loc=<16.944 -0.6265 11.4741>, rot=<0.5141 0.0241 0.4523 -0.7284>}
 dyn_finHeadband_top_r_02 {parent=dyn_finHeadband_top_r_01, loc=<4.9052 0 0>, rot=<0 -0.1966 -0.0715 -0.9779>}
 dyn_finHeadband_top_r_03 {parent=dyn_finHeadband_top_r_02, loc=<5.9672 0 0>, rot=<0 0 0 1>}
 dyn_finHeadband_mid_r_01 {parent=head, loc=<7.8759 -5.1505 14.7017>, rot=<0.2399 0.1322 0.6555 -0.7038>}
 dyn_finHeadband_mid_r_02 {parent=dyn_finHeadband_mid_r_01, loc=<5.8464 0 0>, rot=<0 -0.1435 -0.0384 -0.9889>}
 dyn_finHeadband_mid_r_03 {parent=dyn_finHeadband_mid_r_02, loc=<6.461 0 0>, rot=<0 0 0 1>}
 dyn_finHeadband_top_l_01 {parent=head, loc=<16.944 -0.6265 -11.474>, rot=<-0.4524 -0.0616 0.465 -0.7585>}
 dyn_finHeadband_top_l_02 {parent=dyn_finHeadband_top_l_01, loc=<5.1685 0 0>, rot=<-0.0426 0.1989 -0.0824 -0.9756>}
 dyn_finHeadband_top_l_03 {parent=dyn_finHeadband_top_l_02, loc=<5.7204 0 0>, rot=<0 0 0 1>}
 dyn_finHeadband_mid_l_01 {parent=head, loc=<7.8759 -5.1505 -14.702>, rot=<-0.2252 -0.1354 0.6555 -0.708>}
 dyn_finHeadband_mid_l_02 {parent=dyn_finHeadband_mid_l_01, loc=<5.9553 0 0>, rot=<0 0.1318 -0.0353 -0.9906>}
 dyn_finHeadband_mid_l_03 {parent=dyn_finHeadband_mid_l_02, loc=<6.3307 0 0>, rot=<0 0 0 1>}
 dyn_finHat_bot_l_base {parent=head, loc=<-0.835 -1.458 -10.4651>, rot=<0.2993 -0.7321 0.5636 -0.2383>}
 dyn_finHat_bot_l_01 {parent=dyn_finHat_bot_l_base, loc=<0 0 0>, rot=<0 0 0 1>}
 dyn_finHat_bot_l_02 {parent=dyn_finHat_bot_l_01, loc=<4.5475 -0.4275 -0.1833>, rot=<0 0 0 1>}
 dyn_finHat_bot_l_03 {parent=dyn_finHat_bot_l_02, loc=<6.0641 -0.926 0.5346>, rot=<-0.0007 -0.0294 -0.0807 -0.9963>}
 dyn_finHat_bot_r_base {parent=head, loc=<-0.673 -1.3852 10.3589>, rot=<0.2361 -0.4849 0.7748 -0.33>}
 dyn_finHat_bot_r_01 {parent=dyn_finHat_bot_r_base, loc=<0 0 0>, rot=<0 0 0 1>}
 dyn_finHat_bot_r_02 {parent=dyn_finHat_bot_r_01, loc=<4.6873 0.3159 -0.5498>, rot=<0 0 0 1>}
 dyn_finHat_bot_r_03 {parent=dyn_finHat_bot_r_02, loc=<6.0641 1.0693 0>, rot=<0.0461 -0.0337 0.0831 -0.9949>}
 dyn_finHeadband_bot_r_base {parent=head, loc=<-0.6754 -1.1 10.7704>, rot=<0.2584 0.5142 0.6972 -0.4275>}
 dyn_finHeadband_bot_r_01 {parent=dyn_finHeadband_bot_r_base, loc=<0 0 0>, rot=<0 -0.2588 0 -0.9659>}
 dyn_finHeadband_bot_r_02 {parent=dyn_finHeadband_bot_r_01, loc=<5.6994 -0.1203 1.1054>, rot=<0 -0.178 -0.0314 -0.9835>}
 dyn_finHeadband_bot_r_03 {parent=dyn_finHeadband_bot_r_02, loc=<5.0338 -0.0366 0.3257>, rot=<0.0436 0 0 -0.999>}
 dyn_finHeadband_bot_l_base {parent=head, loc=<-0.8383 -1.0595 -10.4651>, rot=<-0.1852 -0.6232 0.5856 -0.4841>}
 dyn_finHeadband_bot_l_01 {parent=dyn_finHeadband_bot_l_base, loc=<0 0 0>, rot=<-0.2091 0.2527 0.056 -0.943>}
 dyn_finHeadband_bot_l_02 {parent=dyn_finHeadband_bot_l_01, loc=<5.4013 -0.28 -1.338>, rot=<-0.0429 0.1724 -0.0544 -0.9826>}
 dyn_finHeadband_bot_l_03 {parent=dyn_finHeadband_bot_l_02, loc=<5.2765 0.1563 -0.4083>, rot=<-0.0436 0 0 -0.999>}
 clavicle_l {parent=spine_05, loc=<4.3575 -0.1466 -2.3223>, rot=<0.0465 -0.7289 -0.009 -0.6829>}
 clavicle_r {parent=spine_05, loc=<4.3575 -0.1466 2.3223>, rot=<0.7289 0.0465 -0.6829 0.009>}
]
variables [HeadUp {type=double, default=0}]
