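# Physics asset for the face-accessory hat of the OatmealSpreadGolem male medium cosmetic body, defining secondary motion for its fins.
# Left and right top and mid fins are two-capsule chains constrained to the head, while bottom fins are three-capsule chains hung from tiny base-bone capsules.
# Swing and twist limits restrict each link, with twist kept tight.
# Head, upper spine and clavicle capsules supply body volumes around the fins.
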PhysicsAsset /BRCosmetics/Characters/Player/Male/Medium/Bodies/M_MED_OatmealSpreadGolem/Meshes/Parts/M_MED_OatmealSpreadGolem_FaceAcc_HatPhysics
bodies {
 clavicle_l [{shapeType=Capsule, length=11, radius=10.505, shapeName=clavicle_l_capsule, center=<10.5716 0 -1.3333>, rotation=<rot -0.0002 -90 90>}]
 clavicle_r [{shapeType=Capsule, length=12, radius=9.505, shapeName=clavicle_r_capsule, center=<-9.9983 0 1.9629>, rotation=<rot -0.0004 -90 -90>}]
 dyn_finHat_bot_l_01 [{
   shapeType=Capsule
   length=3.5
   radius=3.5
   shapeName=dyn_finHat_bot_l_01_capsule
   center=<5.3568 -0.016 0.8238>
   rotation=<rot 90 -90 -90.0001>
  }]
 dyn_finHat_bot_l_02 [{shapeType=Capsule, length=3.5, radius=3, shapeName=dyn_finHat_bot_l_02_capsule, center=<5.1549 0 1.9671>, rotation=<rot 90 0 -0.0002>}]
 dyn_finHat_bot_l_03 [{
   shapeType=Capsule
   length=4.2461
   radius=2.8147
   shapeName=dyn_finHat_bot_l_03_capsule
   center=<4.2059 -0.3974 1.4976>
   rotation=<rot 0 76.2484 -84.369>
  }]
 dyn_finHat_bot_l_base [{shapeType=Capsule, length=0.505, radius=0.505, shapeName=dyn_finHat_bot_l_base_capsule, rotation=<rot 0 -90 0>}]
 dyn_finHat_bot_r_01 [{
   shapeType=Capsule
   length=1.5
   radius=2.5
   shapeName=dyn_finHat_bot_r_01_capsule
   center=<2.8547 0 -0.3367>
   rotation=<rot 90 -179.9999 -179.9996>
  }]
 dyn_finHat_bot_r_02 [{
   shapeType=Capsule
   length=1.5
   radius=3
   shapeName=dyn_finHat_bot_r_02_capsule
   center=<4.8783 0 0.0234>
   rotation=<rot 90 144.7358 144.7361>
  }]
 dyn_finHat_bot_r_03 [{
   shapeType=Capsule
   length=0.6401
   radius=2.3556
   shapeName=dyn_finHat_bot_r_03_capsule
   center=<4.3726 -0.0865 0.1469>
   rotation=<rot 0.5578 100.007 -96.8353>
  }]
 dyn_finHat_bot_r_base [{shapeType=Capsule, length=0.505, radius=0.505, shapeName=dyn_finHat_bot_r_base_capsule, rotation=<rot 0 -90 0>}]
 dyn_finHat_mid_l_01 [{
   shapeType=Capsule
   length=4.5
   radius=3.5
   shapeName=dyn_finHat_mid_l_01_capsule
   center=<3.0999 -1.3368 1.9463>
   rotation=<rot 78.8803 -162.064 -142.6601>
  }]
 dyn_finHat_mid_l_02 [{shapeType=Capsule, length=4, radius=3.5, shapeName=dyn_finHat_mid_l_02_capsule, center=<3.8083 0 2.4256>, rotation=<rot 90 0 -0.0001>}]
 dyn_finHat_mid_r_01 [{shapeType=Capsule, length=2.5, radius=4, shapeName=dyn_finHat_mid_r_01_capsule, center=<2.0353 1.0309 2.1938>, rotation=<rot 90 0 -19.9998>}]
 dyn_finHat_mid_r_02 [{shapeType=Capsule, length=3, radius=5, shapeName=dyn_finHat_mid_r_02_capsule, center=<3.388 0 2.5965>, rotation=<rot 70.0002 180 -179.9999>}]
 dyn_finHat_top_l_01 [{
   shapeType=Capsule
   length=2.5
   radius=3
   shapeName=dyn_finHat_top_l_01_capsule
   center=<2.883 -0.7028 0.5006>
   rotation=<rot 90 -178.2101 -168.2106>
  }]
 dyn_finHat_top_l_02 [{shapeType=Capsule, length=2, radius=3, shapeName=dyn_finHat_top_l_02_capsule, center=<3.8083 0 1.7059>, rotation=<rot 90 90 89.9996>}]
 dyn_finHat_top_r_01 [{shapeType=Capsule, length=3, radius=2.5, shapeName=dyn_finHat_top_r_01_capsule, center=<2.5821 0.7968 0.9063>, rotation=<rot 90 0 -9.9996>}]
 dyn_finHat_top_r_02 [{shapeType=Capsule, length=4, radius=3, shapeName=dyn_finHat_top_r_02_capsule, center=<2.9786 0 0.3951>, rotation=<rot 80 179.9997 -180>}]
 head [
  {shapeType=Capsule, length=27, radius=11.8227, center=<5.579 4.6888 0>}
  {shapeType=Capsule, length=33.5, radius=9, shapeName=head_capsule, center=<-2.4591 6.1125 0>}
 ]
 spine_05 [{shapeType=Capsule, length=25, radius=15, shapeName=spine_05_capsule, center=<-3.1156 1.06 0>}]
}
constraints [
 {bone1Name=dyn_finHat_bot_l_01, bone2Name=dyn_finHat_bot_l_base, swing1LimitDegrees=30, swing2LimitDegrees=70, twistLimitDegrees=3}
 {bone1Name=dyn_finHat_bot_l_02, bone2Name=dyn_finHat_bot_l_01, swing1LimitDegrees=30, swing2LimitDegrees=70, twistLimitDegrees=3}
 {bone1Name=dyn_finHat_bot_r_01, bone2Name=dyn_finHat_bot_r_base, swing1LimitDegrees=30, swing2LimitDegrees=70, twistLimitDegrees=3}
 {bone1Name=dyn_finHat_bot_r_02, bone2Name=dyn_finHat_bot_r_01, swing1LimitDegrees=30, swing2LimitDegrees=70, twistLimitDegrees=3}
 {bone1Name=dyn_finHat_mid_l_01, bone2Name=head, swing1LimitDegrees=30, swing2LimitDegrees=55, twistLimitDegrees=5}
 {bone1Name=dyn_finHat_mid_l_02, bone2Name=dyn_finHat_mid_l_01, swing1LimitDegrees=30, swing2LimitDegrees=40, twistLimitDegrees=5}
 {bone1Name=dyn_finHat_mid_r_01, bone2Name=head, swing1LimitDegrees=30, swing2LimitDegrees=55, twistLimitDegrees=5}
 {bone1Name=dyn_finHat_mid_r_02, bone2Name=dyn_finHat_mid_r_01, swing1LimitDegrees=30, swing2LimitDegrees=40, twistLimitDegrees=5}
 {bone1Name=dyn_finHat_top_l_01, bone2Name=head, swing1LimitDegrees=30, swing2LimitDegrees=55, twistLimitDegrees=5}
 {bone1Name=dyn_finHat_top_l_02, bone2Name=dyn_finHat_top_l_01, swing1LimitDegrees=30, swing2LimitDegrees=40, twistLimitDegrees=5}
 {bone1Name=dyn_finHat_top_r_01, bone2Name=head, swing1LimitDegrees=30, swing2LimitDegrees=55, twistLimitDegrees=5}
 {bone1Name=dyn_finHat_top_r_02, bone2Name=dyn_finHat_top_r_01, swing1LimitDegrees=20, swing2LimitDegrees=70, twistLimitDegrees=5}
 {bone1Name=dyn_finHat_bot_l_03, bone2Name=dyn_finHat_bot_l_02, swing1LimitDegrees=30, swing2LimitDegrees=70, twistLimitDegrees=3}
 {bone1Name=dyn_finHat_bot_r_03, bone2Name=dyn_finHat_bot_r_02, swing1LimitDegrees=30, swing2LimitDegrees=70, twistLimitDegrees=3}
]
