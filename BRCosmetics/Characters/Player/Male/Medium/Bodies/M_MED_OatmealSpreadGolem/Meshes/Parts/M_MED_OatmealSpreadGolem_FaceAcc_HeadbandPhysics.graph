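# Physics asset for the finned headband face accessory on the OatmealSpreadGolem player body, set up for secondary motion of the fins.
# Six three-bone capsule chains (top, mid and bottom on each side) hang from the head, or from small base bodies for the bottom pair.
# Constraints allow wide swing with very tight twist, looser on most top-fin joints.
# Unconstrained head, neck, clavicle and upper-spine capsules provide body collision shapes.
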PhysicsAsset /BRCosmetics/Characters/Player/Male/Medium/Bodies/M_MED_OatmealSpreadGolem/Meshes/Parts/M_MED_OatmealSpreadGolem_FaceAcc_HeadbandPhysics
bodies {
 clavicle_l [{shapeType=Capsule, length=7, radius=9.505, shapeName=clavicle_l_capsule, center=<12.3287 0 -1.6316>, rotation=<rot -0.0003 -90 100>}]
 clavicle_r [{shapeType=Capsule, length=7, radius=9.505, shapeName=clavicle_r_capsule, center=<-12.2514 0 2.07>, rotation=<rot -0.0005 -90 -80>}]
 dyn_finHeadband_bot_l_01 [{
   shapeType=Capsule
   length=4
   radius=3
   shapeName=dyn_finHeadband_bot_l_01_capsule
   center=<3.6705 0 0>
   rotation=<rot 90 -179.9999 -179.9998>
  }]
 dyn_finHeadband_bot_l_02 [{shapeType=Capsule, length=4, radius=3, shapeName=dyn_finHeadband_bot_l_02_capsule, center=<3.6705 0 0>, rotation=<rot 90 180 -179.9998>}]
 dyn_finHeadband_bot_l_03 [{
   shapeType=Capsule
   length=4
   radius=3
   shapeName=dyn_finHeadband_bot_l_03_capsule
   center=<3.6705 0 0>
   rotation=<rot 90 -179.9998 -179.9996>
  }]
 dyn_finHeadband_bot_l_base [{shapeType=Capsule, length=0.505, radius=0.505, shapeName=dyn_finHeadband_bot_l_base_capsule, rotation=<rot 0 -90 0>}]
 dyn_finHeadband_bot_r_01 [{shapeType=Capsule, length=4, radius=3, shapeName=dyn_finHeadband_bot_r_01_capsule, center=<4.2147 0 0>, rotation=<rot 90 180 -179.9999>}]
 dyn_finHeadband_bot_r_02 [{shapeType=Capsule, length=4, radius=3, shapeName=dyn_finHeadband_bot_r_02_capsule, center=<3.5707 0 0>, rotation=<rot 90 -0.0001 0.0001>}]
 dyn_finHeadband_bot_r_03 [{shapeType=Capsule, length=4, radius=3, shapeName=dyn_finHeadband_bot_r_03_capsule, center=<3.5707 0 0>, rotation=<rot 90 179.9998 -180>}]
 dyn_finHeadband_bot_r_base [{shapeType=Capsule, length=0.505, radius=0.505, shapeName=dyn_finHeadband_bot_r_base_capsule, rotation=<rot 0 -90 0>}]
 dyn_finHeadband_mid_l_01 [{
   shapeType=Capsule
   length=4
   radius=3
   shapeName=dyn_finHeadband_mid_l_01_capsule
   center=<2.6393 0 0>
   rotation=<rot 90 -179.9605 -179.9603>
  }]
 dyn_finHeadband_mid_l_02 [{shapeType=Capsule, length=4, radius=3, shapeName=dyn_finHeadband_mid_l_02_capsule, center=<2.6393 0 0>, rotation=<rot 90 180 -179.9998>}]
 dyn_finHeadband_mid_l_03 [{shapeType=Capsule, length=4, radius=3, shapeName=dyn_finHeadband_mid_l_03_capsule, center=<2.6393 0 0>, rotation=<rot 90 180 -179.9998>}]
 dyn_finHeadband_mid_r_01 [{shapeType=Capsule, length=3, radius=4, shapeName=dyn_finHeadband_mid_r_01_capsule, center=<2.6393 0 0>, rotation=<rot 90 180 -179.9998>}]
 dyn_finHeadband_mid_r_02 [{shapeType=Capsule, length=3, radius=4, shapeName=dyn_finHeadband_mid_r_02_capsule, center=<2.6393 0 0>, rotation=<rot 90 90 90.0002>}]
 dyn_finHeadband_mid_r_03 [{shapeType=Capsule, length=4, radius=3, shapeName=dyn_finHeadband_mid_r_03_capsule, center=<2.6393 0 0>, rotation=<rot 90 90 90.0002>}]
 dyn_finHeadband_top_l_01 [{
   shapeType=Capsule
   length=4
   radius=3
   shapeName=dyn_finHeadband_top_l_01_capsule
   center=<2.6393 0 0>
   rotation=<rot 90 -179.9605 -179.9604>
  }]
 dyn_finHeadband_top_l_02 [{
   shapeType=Capsule
   length=4
   radius=3
   shapeName=dyn_finHeadband_top_l_02_capsule
   center=<2.6393 0 0>
   rotation=<rot 90 -160.5287 -160.5286>
  }]
 dyn_finHeadband_top_l_03 [{shapeType=Capsule, length=4, radius=3, shapeName=dyn_finHeadband_top_l_03_capsule, center=<2.6393 0 0>, rotation=<rot 90 -0.0001 0>}]
 dyn_finHeadband_top_r_01 [{shapeType=Capsule, length=4, radius=3, shapeName=dyn_finHeadband_top_r_01_capsule, center=<2.6393 0 0>, rotation=<rot 90 179.9604 179.9605>}]
 dyn_finHeadband_top_r_02 [{shapeType=Capsule, length=4, radius=3, shapeName=dyn_finHeadband_top_r_02_capsule, center=<2.6393 0 0>, rotation=<rot 90 179.9999 -179.9999>}]
 dyn_finHeadband_top_r_03 [{shapeType=Capsule, length=4, radius=3, shapeName=dyn_finHeadband_top_r_03_capsule, center=<2.6393 0 0>, rotation=<rot 90 90 90.0001>}]
 head [{shapeType=Capsule, length=2.5, radius=11.8227, center=<5.579 4.6888 0>}]
 neck_01 [{shapeType=Capsule, length=9, radius=8.5, shapeName=neck_01_capsule, center=<1.8353 0.3615 0>, rotation=<rot 90 6.7214 6.7215>}]
 spine_05 [{shapeType=Capsule, length=30, radius=12.5, shapeName=spine_05_capsule}]
}
constraints [
 {bone1Name=dyn_finHeadband_bot_l_01, bone2Name=dyn_finHeadband_bot_l_base, swing1LimitDegrees=45, swing2LimitDegrees=45, twistLimitDegrees=5}
 {bone1Name=dyn_finHeadband_bot_l_02, bone2Name=dyn_finHeadband_bot_l_01, swing1LimitDegrees=45, swing2LimitDegrees=45, twistLimitDegrees=5}
 {bone1Name=dyn_finHeadband_bot_l_03, bone2Name=dyn_finHeadband_bot_l_02, swing1LimitDegrees=45, swing2LimitDegrees=45, twistLimitDegrees=5}
 {bone1Name=dyn_finHeadband_bot_r_01, bone2Name=dyn_finHeadband_bot_r_base, swing1LimitDegrees=45, swing2LimitDegrees=45, twistLimitDegrees=5}
 {bone1Name=dyn_finHeadband_bot_r_02, bone2Name=dyn_finHeadband_bot_r_01, swing1LimitDegrees=45, swing2LimitDegrees=45, twistLimitDegrees=5}
 {bone1Name=dyn_finHeadband_bot_r_03, bone2Name=dyn_finHeadband_bot_r_02, swing1LimitDegrees=45, swing2LimitDegrees=45, twistLimitDegrees=5}
 {bone1Name=dyn_finHeadband_mid_l_01, bone2Name=head, swing1LimitDegrees=45, swing2LimitDegrees=45, twistLimitDegrees=5}
 {bone1Name=dyn_finHeadband_mid_l_02, bone2Name=dyn_finHeadband_mid_l_01, swing1LimitDegrees=45, swing2LimitDegrees=45, twistLimitDegrees=5}
 {bone1Name=dyn_finHeadband_mid_l_03, bone2Name=dyn_finHeadband_mid_l_02, swing1LimitDegrees=45, swing2LimitDegrees=45, twistLimitDegrees=5}
 {bone1Name=dyn_finHeadband_mid_r_01, bone2Name=head, swing1LimitDegrees=45, swing2LimitDegrees=45, twistLimitDegrees=5}
 {bone1Name=dyn_finHeadband_mid_r_02, bone2Name=dyn_finHeadband_mid_r_01, swing1LimitDegrees=45, swing2LimitDegrees=45, twistLimitDegrees=5}
 {bone1Name=dyn_finHeadband_mid_r_03, bone2Name=dyn_finHeadband_mid_r_02, swing1LimitDegrees=45, swing2LimitDegrees=45, twistLimitDegrees=5}
 {bone1Name=dyn_finHeadband_top_l_01, bone2Name=head, swing1LimitDegrees=45, swing2LimitDegrees=75, twistLimitDegrees=5}
 {bone1Name=dyn_finHeadband_top_l_02, bone2Name=dyn_finHeadband_top_l_01, swing1LimitDegrees=45, swing2LimitDegrees=75, twistLimitDegrees=5}
 {bone1Name=dyn_finHeadband_top_l_03, bone2Name=dyn_finHeadband_top_l_02, swing1LimitDegrees=45, swing2LimitDegrees=75, twistLimitDegrees=5}
 {bone1Name=dyn_finHeadband_top_r_01, bone2Name=head, swing1LimitDegrees=45, swing2LimitDegrees=45, twistLimitDegrees=5}
 {bone1Name=dyn_finHeadband_top_r_02, bone2Name=dyn_finHeadband_top_r_01, swing1LimitDegrees=45, swing2LimitDegrees=60, twistLimitDegrees=5}
 {bone1Name=dyn_finHeadband_top_r_03, bone2Name=dyn_finHeadband_top_r_02, swing1LimitDegrees=45, swing2LimitDegrees=60, twistLimitDegrees=5}
]
