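# Animation Blueprint for the OatmealSpreadGolem character's face accessory, adding physics to a hat or headband.
# Copies the parent pose, layers in Body Mesh from the upper arms, then runs a Control Rig and three chained RigidBody nodes: hat, front-end hat, headband.
# On customization load, a style metatag and front-end state set per-accessory On values, which tactical sprinting adjusts.
# Each update computes GravityOverride through a deprecated helper.
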AnimBlueprint /BRCosmetics/Characters/Player/Male/Medium/Bodies/M_MED_OatmealSpreadGolem/Meshes/Parts/M_MED_OatmealSpreadGolem_FaceAcc_AnimBP
target_skeleton=&/M_MED_OatmealSpreadGolem_FaceAcc_Skeleton
graphs [
 AnimGraph {nodes [
   Copy Pose From Mesh() {bUseAttachedParent=true}
   n1 = Get Body Mesh() -> (Body Mesh:Skeletal Mesh Component Object Reference)
   n2 = Control Rig(Source=Layered blend per bone(
     BasePose=Copy Pose From Mesh() {bCopyCurves=true, bUseAttachedParent=true}
     BlendPoses_0=Copy Pose From Mesh(SourceMeshComponent=@n1) {bUseAttachedParent=true}
    ) {layerSetup [{branchFilters [{boneName=upperarm_l}, {boneName=upperarm_r}]}]}) {
    alphaBoolBlend {blendOption=Linear}
    controlRigAssetReference {blueprintRigClass=&/M_MED_OatmealSpreadGolem_FaceAcc_CtrlRig.M_MED_OatmealSpreadGolem_FaceAcc_CtrlRig_C}
   }
   n3 = RigidBody!(
    ComponentPose=Local To Component(LocalPose=@n2)
    Alpha
    ExternalForce
    ComponentLinearAccScale:Vector="0,0,0"
    ComponentLinearVelScale:Vector="0,0,0"
   ) -> (Pose) {
    bClampLinearTranslationLimitToRefPose=true
    bOverrideWorldGravity=true
    overridePhysicsAsset=&/M_MED_OatmealSpreadGolem_FaceAcc_HatPhysics
    simulationSpace=BaseBoneSpace
    alphaBoolBlend {blendOption=Linear}
    baseBoneRef {boneName=neck_02}
    componentAppliedLinearAccClamp=<3000 3000 3000>
    componentLinearAccScale=<0.3 0.3 0.3>
    componentLinearVelScale=<0.35 0.35 0.35>
    overrideWorldGravity {z=-980}
    simSpaceSettings {maxAngularAcceleration=1500, maxAngularVelocity=1500, maxLinearAcceleration=500, maxLinearVelocity=500}
   }
   n4 = RigidBody(ComponentPose=@n3, ComponentLinearAccScale="0,0,0", ComponentLinearVelScale="0,0,0") {
    bClampLinearTranslationLimitToRefPose=true
    bOverrideWorldGravity=true
    overridePhysicsAsset=&/M_MED_OatmealSpreadGolem_FaceAcc_FE_HatPhysics
    simulationSpace=BaseBoneSpace
    alphaBoolBlend {blendOption=Linear}
    baseBoneRef {boneName=neck_02}
    componentAppliedLinearAccClamp=<3000 3000 3000>
    componentLinearAccScale=<0.3 0.3 0.3>
    componentLinearVelScale=<0.35 0.35 0.35>
    overrideWorldGravity {z=-980}
    simSpaceSettings {maxAngularAcceleration=1500, maxAngularVelocity=1500, maxLinearAcceleration=500, maxLinearVelocity=500}
   }
   n5 = RigidBody(ComponentPose=@n4, ComponentLinearAccScale="0.8,0.8,0.35", ComponentLinearVelScale="0.8,0.8,0.35") {
    bClampLinearTranslationLimitToRefPose=true
    bOverrideWorldGravity=true
    overridePhysicsAsset=&/M_MED_OatmealSpreadGolem_FaceAcc_HeadbandPhysics
    simulationSpace=BaseBoneSpace
    alphaBoolBlend {blendOption=Linear}
    baseBoneRef {boneName=neck_02}
    componentAppliedLinearAccClamp=<3000 3000 3000>
    componentLinearAccScale=<0.8 0.8 0.35>
    componentLinearVelScale=<0.8 0.8 0.35>
    overrideWorldGravity {z=-980}
    simSpaceSettings {
     maxAngularAcceleration=1500
     maxAngularVelocity=1500
     maxLinearAcceleration=500
     maxLinearVelocity=500
     externalAngularVelocity=<0.2 0.2 0.2>
    }
   }
   Output Pose(Result=Component To Local(ComponentPose=@n5)) {layerGroup=None}
  ]}
 EventGraph {nodes [
   Event BlueprintUpdateAnimation()
   > n1 = Gravity Override State Function DEPRECATED(CharacterPartAnimInstance=Self-Reference(), GameTimeInSeconds=GetGameTimeInSeconds(), GravityOverrideParams=Get Gravity Override Params())
   > Set GravityOverride(GravityOverride=@n1)
   Event OnCharacterCustomizationLoaded()
   > n2 = Sequence()
   > n3 = Set Hat Metatag(Hat Metatag:bool=CharacterItemDefinitionHasAnyMetaTag(Tags=Make Array("[0]"=(TagName="Cosmetics.MetaTags.Style.OptionA")))) -> (Output_Get:bool)
   n4 = Branch(execute=@n2.then_1, Condition=@n3)
   > Set Hat On(Hat On:double=1) -> (Output_Get:double)
   > Set Headband On(Headband On:double=0) -> (Output_Get:double)
   > n5 = Branch(Condition=Get bIsInFrontEnd())
   > Set Hat On(Hat On=0)
   > Set Hat On Frontend(Hat On Frontend:double=1) -> (Output_Get:double)
   Set Hat On(execute=@n5.else, Hat On=1)
   > Set Hat On Frontend(Hat On Frontend=0)
   Set Hat On(execute=@n4.else, Hat On=0)
   > Set Headband On(Headband On=1)
   > Set Hat On Frontend(Hat On Frontend=0)
   Event OnBeginTacticalSprinting() {is=title K2Node_Event}
   n6 = Get Hat Metatag() -> (Hat Metatag:bool)
   > n7 = Branch(Condition=@n6)
   > Set Hat On(Hat On=0.7)
   > Set Headband On(Headband On=0)
   Set Hat On(execute=@n7.else, Hat On=0)
   > Set Headband On(Headband On=0.5)
   Event BlueprintInitializeAnimation()
   > Set Skel_Mesh(Skel_Mesh=GetOwnerSkeletalMeshComponent())
   Event OnEndTacticalSprinting() {is=title K2Node_Event}
   > n8 = Branch(Condition=@n6)
   > Set Hat On(Hat On=1)
   > Set Headband On(Headband On=0)
   Set Hat On(execute=@n8.else, Hat On=0)
   > Set Headband On(Headband On=1)
   "Gravity Override State Function (Nativized)"!(
    execute
    CharacterPartAnimInstance:Custom Character Part Anim Instance Object Reference
    GravityOverrideParameterSet:Fort Gravity Override Parameter Set Structure
    RigidBodyAnimNodeParameterSet:Fort Rigid Body Anim Node Parameter Set Structure
    DebugDraw:bool=false
    GameTimeInSeconds:float=0
    DebugJointName:Name=root
    StateLogicLayerName:Name=none
   ) -> (
    then
    OutGravityOverride:Vector
    OutCurrentState:Name=None
    OutGravityOverrideParameters:Fort Gravity Override Parameters Structure
    OutRigidBodyAnimNodeParameters:Fort Rigid Body Anim Node Parameters Structure
   )
  ]}
]
variables [
 Body Mesh {type=SkeletalMeshComponent}
 Gravity Override Params {type=GravityOverrideParamsStruct}
 GravityOverride {type=Vector}
 Hat Component Linear Acc Scale {type=Vector, default=<0.3 0.3 0.3>}
 Hat Component Linear Vel Scale {type=Vector, default=<0.35 0.35 0.35>}
 Hat Metatag {type=boolean}
 Hat On {type=number}
 Hat On Frontend {type=number}
 "Headband  Component Linear Acc Scale" {type=Vector, default=<0.8 0.8 0.35>}
 Headband Component Linear Vel Scale {type=Vector, default=<0.8 0.8 0.35>}
 Headband On {type=number}
 Skel_Mesh {type=SkeletalMeshComponent}
]
